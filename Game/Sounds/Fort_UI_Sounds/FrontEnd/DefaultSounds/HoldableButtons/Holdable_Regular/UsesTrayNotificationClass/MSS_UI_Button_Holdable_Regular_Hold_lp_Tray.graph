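# Stereo MetaSound source for the hold loop of a regular holdable front-end UI button, routed to the tray notification sound class.
# It passes all inputs into the shared MSS_PlayRandomSoundLooping_Stereo source, supplying one hold-loop wave, slight volume variation, fixed pitch and short attack/release times.
# That node drives both audio channels and the finished trigger.
# Overridden concurrency allows a single instance, blocks new ones and sets a retrigger time.
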MetaSoundSource /Game/Sounds/Fort_UI_Sounds/FrontEnd/DefaultSounds/HoldableButtons/Holdable_Regular/UsesTrayNotificationClass/MSS_UI_Button_Holdable_Regular_Hold_lp_Tray
inputs [
 Attack Time {type=Time}
 bUseRandomStartTime {type=Bool}
 Loop {type=Bool}
 Max Pitch {type=Float}
 Max Volume {type=Float}
 Min Pitch {type=Float}
 Min Volume {type=Float}
 Release Time {type=Time}
 Sounds {type=WaveAsset:Array}
 UE.Source.OnPlay {type=Trigger}
]
interfaces [UE.OutputFormat.Stereo:1.0, UE.Source.OneShot:1.0, UE.Source:1.1]
outputs [
 "UE.OutputFormat.Stereo.Audio:0" {type=Audio}
 "UE.OutputFormat.Stereo.Audio:1" {type=Audio}
 UE.Source.OneShot.OnFinished {type=Trigger}
]
pages [Default {
  input_defaults {
   Attack Time=0.2
   Loop="<MetaSoundBuilderResult.SUCCEEDED: 0>"
   Max Pitch=1
   Max Volume=1
   Min Pitch=1
   Min Volume=0.9
   Release Time=0.2
   UE.Source.OnPlay=null
   bUseRandomStartTime=null
   Sounds [/Game/Sounds/Fort_UI_Sounds/FrontEnd/DefaultSounds/HoldableButtons/wav/UI_Button_Press_Hold_Normal_Hold_lp]
  }
  nodes [MSS_PlayRandomSoundLooping_Stereo_0 = /FortMetasound/Sources/MSS_PlayRandomSoundLooping_Stereo(
    Attack Time=^Attack Time
    Loop=^Loop
    Max Pitch=^Max Pitch
    Max Volume=^Max Volume
    Min Pitch=^Min Pitch
    Min Volume=^Min Volume
    Release Time=^Release Time
    Sounds=^Sounds
    UE.Source.OnPlay=^UE.Source.OnPlay
    bUseRandomStartTime=^bUseRandomStartTime
   )]
  links [
   {dst="@UE.OutputFormat.Stereo.Audio:0", src=MSS_PlayRandomSoundLooping_Stereo_0.UE.OutputFormat.Stereo.Audio:0}
   {dst="@UE.OutputFormat.Stereo.Audio:1", src=MSS_PlayRandomSoundLooping_Stereo_0.UE.OutputFormat.Stereo.Audio:1}
   {dst="@UE.Source.OneShot.OnFinished", src=MSS_PlayRandomSoundLooping_Stereo_0.UE.Source.OneShot.OnFinished}
  ]
 }]
properties {
 assetImportData=&self:AssetImportData
 bOverrideConcurrency=true
 numChannels=2
 outputFormat=Stereo
 soundClassObject=/Game/Sounds/FortSoundClassesAndModes/Fort_Application_TrayNotification
 volume=0.504
 concurrencyOverrides {maxCount=1, resolutionRule=PreventNew, retriggerTime=0.5}
}
